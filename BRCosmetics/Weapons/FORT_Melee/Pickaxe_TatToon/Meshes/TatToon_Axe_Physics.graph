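# Physics asset for the TatToon pickaxe, a two-handed melee weapon cosmetic with a dangling chain.
# Capsule bodies cover the handle and six chain bones, from dyn_chaine_base through dyn_chaine_5, most of them zero-length and thus effectively spheres.
# Five constraints link each chain bone to the previous one with limited swing and twist, forming a serial chain.
# The handle has two capsules and no constraint to the chain.
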PhysicsAsset /BRCosmetics/Weapons/FORT_Melee/Pickaxe_TatToon/Meshes/TatToon_Axe_Physics
bodies {
 Melee_TwoHanded_Handle [
  {
   shapeType=Capsule
   length=27.0375
   radius=2.1819
   shapeName=Melee_TwoHanded_Handle_capsule
   center=<-0.2075 -0.4706 -5.4001>
   rotation=<rot 0.0482 -90 0.0452>
  }
  {shapeType=Capsule, length=0, radius=3.75, shapeName=Melee_TwoHanded_Handle_capsule, center=<0 0 -23.1691>}
 ]
 dyn_chaine_1 [{shapeType=Capsule, length=0, radius=1.2592, shapeName=dyn_chaine_1_capsule, center=<1.1684 -0.0149 0.032>, rotation=<rot 0 89.5635 -68.7819>}]
 dyn_chaine_2 [{
   shapeType=Capsule
   length=0
   radius=1.8017
   shapeName=dyn_chaine_2_capsule
   center=<2.6553 0.0598 -0.0877>
   rotation=<rot 0 -83.5435 -87.3701>
  }]
 dyn_chaine_3 [{
   shapeType=Capsule
   length=0
   radius=1.7706
   shapeName=dyn_chaine_3_capsule
   center=<3.1165 0.0184 -0.071>
   rotation=<rot 0 -92.9144 -83.4785>
  }]
 dyn_chaine_4 [{
   shapeType=Capsule
   length=0
   radius=1.7509
   shapeName=dyn_chaine_4_capsule
   center=<1.8579 -0.1254 0.153>
   rotation=<rot 7.9127 -0.5885 -90.6946>
  }]
 dyn_chaine_5 [{
   shapeType=Capsule
   length=7.2631
   radius=3.0851
   shapeName=dyn_chaine_5_capsule
   center=<4.3165 0.0014 0.2225>
   rotation=<rot -0.0001 1.0629 -89.4162>
  }]
 dyn_chaine_base [{shapeType=Capsule, length=0, radius=1, shapeName=dyn_chaine_base_capsule, rotation=<rot 0 -90 0>}]
}
constraints [
 {bone1Name=dyn_chaine_1, bone2Name=dyn_chaine_base, swing1LimitDegrees=60, swing2LimitDegrees=60, twistLimitDegrees=25}
 {bone1Name=dyn_chaine_2, bone2Name=dyn_chaine_1, swing1LimitDegrees=40, swing2LimitDegrees=40, twistLimitDegrees=30}
 {bone1Name=dyn_chaine_3, bone2Name=dyn_chaine_2, swing1LimitDegrees=45, swing2LimitDegrees=45, twistLimitDegrees=35}
 {bone1Name=dyn_chaine_4, bone2Name=dyn_chaine_3, swing1LimitDegrees=45, swing2LimitDegrees=45, twistLimitDegrees=45}
 {bone1Name=dyn_chaine_5, bone2Name=dyn_chaine_4, swing1LimitDegrees=45, swing2LimitDegrees=45, twistLimitDegrees=45}
]
